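# Editor utility widget that batch-adds SoundWaves from a typed content path into the open Level Sequence's subsequences.
# It maps each sound to a code sliced from its name, then matches that against subsection sequence names.
# Matches get a locked audio track in a new folder, starting at the frame parsed from the name's last four characters.
# Embedded Python sets sound loading and compression, and the sequence clock switches to Audio.
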EditorUtilityWidgetBlueprint /AnimInEdTools/CineAudioAdder/EUW_CineAddAudio
graphs [
 EventGraph {nodes [
   "On Clicked (EditorUtilityButton_45)"() {is=title K2Node_ComponentBoundEvent}
   > GetAudioFiles()
   > n1 = GetCurrentLevelSequence!(execute) -> (then, ReturnValue:Level Sequence Object Reference)
   > n2 = GetTracks!(execute, Sequence:Movie Scene Sequence Object Reference=@n1) -> (then, ReturnValue:Array of Movie Scene Track Object References)
   > n3 = For Each Loop(Array:Array of Movie Scene Track Object References=@n2) -> (Array Element:Movie Scene Track Object Reference)
   Clear!(
    execute=@n3.Completed
    TargetMap:Map of Sound Wave Object References to Strings=Get Audio() -> (Audio:Map of Sound Wave Object References to Strings)
   ) -> (then) {is=title K2Node_CallFunction}
   n4 = Cast To MovieSceneSubTrack!(execute=@n3.LoopBody, Object=@n3) -> (then, CastFailed, AsMovie Scene Sub Track:Movie Scene Sub Track Object Reference) {is=title K2Node_DynamicCast}
   > n5 = GetSections!(execute, Track:Movie Scene Track Object Reference=@n4) -> (then, ReturnValue:Array of Movie Scene Section Object References)
   > n6 = For Each Loop(Array:Array of Movie Scene Section Object References=@n5) -> (Array Element:Movie Scene Section Object Reference)
   > n7 = Cast To MovieSceneSubSection!(execute, Object=@n6) -> (then, CastFailed, AsMovie Scene Sub Section:Movie Scene Sub Section Object Reference) {is=title K2Node_DynamicCast}
   n8 = GetSequence!(self:Movie Scene Sub Section Object Reference=@n7) -> (ReturnValue:Movie Scene Sequence Object Reference)
   > n9 = Cast To LevelSequence!(execute, Object=@n8) -> (then, CastFailed, AsLevel Sequence:Level Sequence Object Reference) {is=title K2Node_DynamicCast}
   > n10 = Values!(execute, TargetMap:Map of Sound Wave Object References to Strings=Get Audio()) -> (then, Values:Array of Strings)
   > n11 = For Each Loop(Array:Array of Strings=@n10) -> (Array Element:String)
   UnlockFolderGen(execute=@n11.Completed)
   Branch(execute=@n11.LoopBody, Condition=Contains(SearchIn=GetDisplayName(Object=@n8), Substring=@n11))
   > n12 = Sequence()
   > Do Once(Reset=@n17.then)
   n13 = Get AudioPath() -> (AudioPath:Text)
   > n14 = AddRootFolderToSequence!(
    execute
    Sequence:Movie Scene Sequence Object Reference=@n9
    NewFolderName:String=Split!(InPath:String="To String (Text)"(InText=@n13)) -> (PathPart:String, FilenamePart:String, ExtensionPart:String).FilenamePart
   ) -> (then, ReturnValue:Movie Scene Folder Object Reference)
   n15 = Keys!(execute=@n12.then_1, TargetMap:Map of Sound Wave Object References to Strings=Get Audio()) -> (then, Keys:Array of Sound Wave Object References)
   > SetAudioFiles!(
    execute
    self
    Folder:Movie Scene Folder Object Reference=@n14
    Audio:Sound Wave Object Reference="Get (a copy)"(Array:Array of Sound Wave Object References=@n15, Dimension 1=@n11.Array Index) -> (Output:Sound Wave Object Reference)
    Seq:Level Sequence Object Reference=@n9
   ) -> (then)
   n16 = "On Text Changed (EditorUtilityEditableTextBox_91)"!() -> (OutputDelegate, then, Text:"Text (by ref)") {is=title K2Node_ComponentBoundEvent}
   > Set AudioPath(AudioPath:Text=@n16.Text) -> (Output_Get:Text)
   n17 = UnlockFolderGen() {is=title K2Node_CustomEvent}
   Event PreConstruct()
   Event Construct()
   Event Tick()
  ]}
 GetAudioFiles {nodes [
   GetAudioFiles() {is=title K2Node_FunctionEntry}
   n1 = Get AudioPath()
   > n2 = GetAssetsByPath!(
    execute
    self:Asset Registry Interface=GetAssetRegistry()
    PackagePath:Name=String To Name(InString="To String (Text)"(InText=@n1))
    bRecursive:bool=false
    bIncludeOnlyOnDiskAssets:bool=false
   ) -> (then, OutAssetData:Array of Asset Data Structures, ReturnValue:bool)
   > n3 = For Each Loop!(Exec, Array:Array of Asset Data Structures=@n2) -> (
    LoopBody
    Array Element_PackageName:Name
    Array Element_PackagePath:Name
    Array Element_AssetName:Name
    Array Element_AssetClassPath:Top Level Asset Path Structure
    Array Index
    Completed
   )
   > n4 = LoadAsset(AssetPath=Append!(A="To String (Name)"(InName=@n3), B=., C:String="To String (Name)"(InName=@n3.Array Element_AssetName)) -> (ReturnValue))
   > n5 = Cast To SoundWave!(execute, Object=@n4) -> (then, CastFailed, AsSound Wave:Sound Wave Object Reference) {is=title K2Node_DynamicCast}
   n6 = GetDisplayName(Object=@n5)
   > Add!(
    execute
    TargetMap:Map of Sound Wave Object References to Strings=Get Audio()
    Key:"Sound Wave Object Reference (by ref)"=@n5
    Value:"String (by ref)"=Left!(SourceString:String=Right!(SourceString:String=@n6, Count:int=11) -> (ReturnValue:String), Count:int=5) -> (ReturnValue:String)
   ) -> (then) {is=title K2Node_CallFunction}
  ]}
 SetAudioFiles {nodes [
   n1 = GetTracks()
   > n2 = For Each Loop!(Exec, Array:Array of Movie Scene Track Object References=@n1) -> (LoopBody, Array Element:Movie Scene Track Object Reference, Array Index, Completed)
   > n3 = GetSections(Track=@n2)
   > n4 = For Each Loop(Array:Array of Movie Scene Section Object References=@n3) -> (Array Element:Movie Scene Section Object Reference)
   > n5 = Cast To MovieSceneAudioSection!(execute, Object=@n4) -> (then, CastFailed, AsMovie Scene Audio Section:Movie Scene Audio Section Object Reference) {is=title K2Node_DynamicCast}
   > PrintString(InString=Hello)
   SetAudioFiles!() -> (then, Folder:Movie Scene Folder Object Reference, Audio:Sound Wave Object Reference, Seq:Level Sequence Object Reference) {is=title K2Node_FunctionEntry}
   > n6 = Sequence!(execute) -> (then_0, then_1, then_2:Exec)
   Execute Python Script!(
    execute=@n6.then_1
    PythonScript="import unreal\r\n\r\ndef configure_sound_wave(sound_wave: unreal.SoundWave):\r\n    if not sound_wave:\r\n        unreal.log_error(\"Invalid SoundWave.\")\r\n        return\r\n\r\n    sound_wave.modify()\r\n\r\n    ## Loading Behavior Override -> Prime On Load\r\n    sound_wave.set_editor_property(\r\n        \"loading_behavior\",\r\n        unreal.SoundWaveLoadingBehavior.PRIME_ON_LOAD\r\n    )\r\n\r\n    ## Compression Type -> Bink Audio\r\n    sound_wave.set_editor_property(\r\n        \"sound_asset_compression_type\",\r\n        unreal.SoundAssetCompressionType.BINK_AUDIO\r\n    )\r\n\r\n    unreal.EditorAssetLibrary.save_loaded_asset(sound_wave)\r\n\r\n    unreal.log(f\"Configured {sound_wave.get_name()}\")\r\n\r\n## Example:\r\nsound = soundwave\r\nconfigure_sound_wave(sound)"
    soundwave:Sound Wave Object Reference=Get Audio() -> (Audio:Sound Wave Object Reference)
   ) -> (then, ReturnValue) {inputs [soundwave]}
   > SaveLoadedAsset(AssetToSave=Get Audio(), bOnlyIfIsDirty=true)
   > n7 = AddTrack!(
    execute
    Sequence:Movie Scene Sequence Object Reference=Get Seq() -> (Seq:Level Sequence Object Reference)
    TrackType:Movie Scene Track Class Reference=/Script/MovieSceneTracks.MovieSceneAudioTrack
   ) -> (then, ReturnValue:Movie Scene Audio Track Object Reference)
   > AddChildTrack!(
    execute
    Folder:Movie Scene Folder Object Reference=Get Folder() -> (Folder:Movie Scene Folder Object Reference)
    InTrack:Movie Scene Track Object Reference=@n7
   ) -> (then, ReturnValue:bool)
   n8 = Get Audio()
   n9 = GetDisplayName(Object=@n8)
   > SetDisplayName!(execute, Track:Movie Scene Track Object Reference=@n7, InName:"Text (by ref)"="To Text (String)"(InString=@n9)) -> (then)
   > n10 = AddSection!(execute, Track:Movie Scene Track Object Reference=@n7) -> (then, ReturnValue:Movie Scene Section Object Reference)
   > n11 = Cast To MovieSceneAudioSection(Object=@n10)
   n12 = String To Integer(InString=Right(SourceString=@n9, Count=4))
   n13 = float * float(A=Get Duration!(self:Sound Base Object Reference=@n8) -> (Duration:float), B:int=Get SeqFPS() -> (SeqFPS:int=0)) {is=title K2Node_PromotableOperator}
   > SetRange!(execute, Section:Movie Scene Section Object Reference=@n11, StartFrame:int=@n12, EndFrame:int=int + int(A=@n12, B=Truncate(A=@n13))) -> (then)
   > SetSound!(execute, self:Movie Scene Audio Section Object Reference=@n11, InSound:Sound Base Object Reference=Get Audio()) -> (then)
   > SetIsLocked!(execute, self:Movie Scene Section Object Reference=@n11, bInIsLocked:bool=true) -> (then)
   SaveLoadedAsset(execute=@n6.then_2, AssetToSave=Get Seq(), bOnlyIfIsDirty=true)
   n14 = GetPlaybackStart!(execute=@n6.then_0, Sequence:Movie Scene Sequence Object Reference=Get Seq()) -> (then, ReturnValue:int=0)
   > Set OriginalPlayBackStart(OriginalPlayBackStart:int=@n14) -> (Output_Get:int=0)
   > SetClockSource!(execute, InSequence:Movie Scene Sequence Object Reference=Get Seq(), InClockSource:EUpdateClockSource Enum=Audio) -> (then)
   > n15 = GetDisplayRate!(execute, Sequence:Movie Scene Sequence Object Reference=Get Seq()) -> (then, ReturnValue_Numerator:int=0, ReturnValue_Denominator:int=0)
   > Set SeqFPS(SeqFPS:int=@n15) -> (Output_Get:int=0)
   GetSound!(self:Movie Scene Audio Section Object Reference=@n5) -> (ReturnValue:Sound Base Object Reference)
   Get Seq()
  ]}
]
variables [
 Audio {default {}}
 AudioPath {type=string, default=""}
 SeqRefs {type="array<LevelSequence>", default []}
]
widgets {
 info {rootWidgetClass=&CanvasPanel, widgetCount=5}
 widgets [
  CanvasPanel_86 {widgetClassPath=&CanvasPanel, widget=CanvasPanel_86}
  TextBlock_219 {
   widgetClassPath=&TextBlock
   parent=CanvasPanel_86
   widget=TextBlock_219
   properties {
    justification=Center
    text=Audio Path
    font {
     bForceMonospaced=false
     bMaterialIsStencil=false
     fontMaterial=null
     fontObject=&Roboto
     letterSpacing=0
     monospacedWidth=1
     size=9
     skewAmount=0
     typefaceFontName=Bold
     outlineSettings {
      bApplyOutlineToDropShadows=false
      bMiteredCorners=false
      bSeparateFillAlpha=false
      outlineMaterial=null
      outlineSize=0
      outlineColor=<rgba 0 0 0 1>
     }
    }
   }
   slot {class=CanvasPanelSlot, properties {layoutData {alignment=<0 0>, anchors {maximum=<0 0>, minimum=<0 0>}, offsets=<ltrb 0 8 100 13.984>}}}
  }
  EditorUtilityEditableTextBox_91 {
   widgetClassPath=&EditorUtilityEditableTextBox
   bIsVariable=true
   parent=CanvasPanel_86
   widget=EditorUtilityEditableTextBox_91
   slot {
    class=CanvasPanelSlot
    properties {layoutData {alignment=<0 0>, anchors {maximum=<1 0>, minimum=<0 0>}, offsets=<ltrb 102.1021 0 0 31.1422>}}
   }
  }
  EditorUtilityButton_45 {
   widgetClassPath=&EditorUtilityButton
   parent=CanvasPanel_86
   widget=EditorUtilityButton_45
   slot {
    class=CanvasPanelSlot
    properties {bAutoSize=true, layoutData {alignment=<0 0>, anchors {maximum=<1 1>, minimum=<0 0>}, offsets=<ltrb 0 34.034 0 0>}}
   }
  }
  TextBlock_81 {
   widgetClassPath=&TextBlock
   parent=EditorUtilityButton_45
   widget=TextBlock_81
   properties {
    justification=Center
    text="Open root sequence\r\ncopy n paste audio location from content browser into text field above\r\nclick this button\r\n"
    font {
     bForceMonospaced=false
     bMaterialIsStencil=false
     fontMaterial=null
     fontObject=&Roboto
     letterSpacing=0
     monospacedWidth=1
     size=9
     skewAmount=0
     typefaceFontName=Bold
     outlineSettings {
      bApplyOutlineToDropShadows=false
      bMiteredCorners=false
      bSeparateFillAlpha=false
      outlineMaterial=null
      outlineSize=0
      outlineColor=<rgba 0 0 0 1>
     }
    }
   }
   slot {class=ButtonSlot, properties {horizontalAlignment=HAlign_Fill, verticalAlignment=VAlign_Fill}}
  }
 ]
}
